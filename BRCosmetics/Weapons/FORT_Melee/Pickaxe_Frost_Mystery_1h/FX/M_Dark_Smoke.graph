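# Unlit translucent particle sprite material for dark, animated smoke in the one-handed Frost Mystery pickaxe's FX.
# Panning caustics in radial coordinates and scrolling noise modulate sphere-mask radii, forming two procedural puff layers, the second carved by subtraction.
# Both layers drive depth-faded opacity, but only the first brightens black toward Color, tinted by particle color.
# Dynamic parameters supply per-particle time offset, radial tiling and fade distance.
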Material /BRCosmetics/Weapons/FORT_Melee/Pickaxe_Frost_Mystery_1h/FX/M_Dark_Smoke
graph {
 properties {
  bUsedWithNiagaraSprites=true
  bUsedWithParticleSprites=true
  blendMode=BLEND_Translucent
  pixelDepthOffsetMode=PDOM_Legacy
  shadingModel=MSM_Unlit
  usedShadingModels=Unlit
 }
 nodes [
  n1 = MaterialFunctionCall(Float2=MaterialFunctionCall() {materialFunction=&VectorToRadialValue}.Radial Coordinates) {materialFunction=&BreakOutFloat2Components}
  n2 = DynamicParameter() {defaultValue=<rgba 1 0.3 1 1>, paramNames [Random,U_Tile,Param3,Param4]}
  n3 = Multiply(A=@n1.R, B=@n2.U_Tile) {constB=0.3}
  n4 = Power(Base=Multiply(A=@n1.G) {constB=0.05}) {constExponent=1.2}
  n5 = Add(A=DynamicParameter() {defaultValue=<rgba 1 0.3 1 1>, paramNames [Random,U_Tile,Param3,Param4]}.Random, B=Time())
  n6 = TextureSample(UVs=Panner(Coordinate=AppendVector(A=@n3, B=@n4), Time=@n5) {speedX=-0.01, speedY=-0.04}) {texture=/Game/Global/Textures/Default/Caustics/T_Caustics_Softer}
  n7 = TextureSample(UVs=Panner(Coordinate=TextureCoordinate() {uTiling=0.5, vTiling=0.5}, Time=@n5) {speedX=0.05, speedY=0.05}) {samplerType=SAMPLERTYPE_Grayscale, texture=&LowResBlurredNoise}
  n8 = Multiply(
   A=SphereMask(A=TextureCoordinate(), B=Constant() {r=0.5}, Radius=LinearInterpolate(Alpha=@n6.R) {constB=0.7}) {attenuationRadius=0.5, hardnessPercent=0}
   B=SphereMask(A=TextureCoordinate(), B=Constant() {r=0.5}, Radius=LinearInterpolate(Alpha=@n7.R) {constA=0.2, constB=0.8}) {attenuationRadius=0.5, hardnessPercent=20}
  )
  n9 = MaterialFunctionCall(Float2=MaterialFunctionCall() {materialFunction=&VectorToRadialValue}.Radial Coordinates) {materialFunction=&BreakOutFloat2Components}
  n10 = Multiply(A=@n9.G) {constB=0.1}
  n11 = Panner(Coordinate=AppendVector(A=Multiply(A=@n9.R, B=@n2.U_Tile) {constB=0.3}, B=Power(Base=@n10) {constExponent=1.6}), Time=@n5) {speedX=0.01, speedY=-0.04}
  n12 = SphereMask(
   A=TextureCoordinate()
   B=Constant() {r=0.5}
   Radius=LinearInterpolate(Alpha=TextureSample(UVs=@n11) {texture=&T_Caustics_Softer}.R)
  ) {attenuationRadius=0.5, hardnessPercent=0}
  n13 = Panner(Coordinate=TextureCoordinate() {uTiling=0.5, vTiling=0.5}, Time=@n5) {speedX=0.05, speedY=0.05}
  n14 = SphereMask(
   A=TextureCoordinate()
   B=Constant() {r=0.5}
   Radius=LinearInterpolate(Alpha=TextureSample(UVs=@n13) {texture=&T_NoisePack_01}.R) {constA=0.4, constB=0}
  ) {attenuationRadius=0.5, hardnessPercent=0}
  n15 = TextureSample(UVs=Panner(Coordinate=TextureCoordinate() {uTiling=0.5, vTiling=0.5}, Time=@n5) {speedX=0.05, speedY=0.05}) {samplerType=SAMPLERTYPE_Grayscale, texture=&LowResBlurredNoise}
  n16 = Multiply(
   A=Saturate(None=Subtract(A=@n12, B=@n14))
   B=SphereMask(A=TextureCoordinate(), B=Constant() {r=0.5}, Radius=LinearInterpolate(Alpha=@n15.R) {constA=0.2}) {attenuationRadius=0.5, hardnessPercent=60}
  )
  n17 = ParticleColor()
  DepthFade(
   Opacity=Multiply(A=Add(A=@n8, B=@n16), B=@n17.A)
   FadeDistance=LinearInterpolate(Alpha=DynamicParameter() {defaultValue=<rgba 1 0.3 1 1>, paramNames [Random,U_Tile,Param3,Param4]}.Param4) {constB=80}
  ) {fadeDistanceDefault=80}
  LinearInterpolate(Alpha=@n6.R) {constA=0.5, constB=0}
  n18 = VectorParameter() {parameterName=Color, defaultValue=<rgba 0.2526 0.2539 0.2552 1>}
  MaterialFunctionCall(Emissive=Multiply(A=@n17.RGB, B=LinearInterpolate(B=@n18.RGB, Alpha=@n8))) {materialFunction=&MF_FX_EmissiveAdjustments}
  TextureCoordinate() {uTiling=0.5, vTiling=0.5}
 ]
}
