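# Masked, lit particle material for glass-like shatter shards, in the effects of the Frost Mystery one-handed pickaxe.
# A 2x2 flipbook picks a shard per particle by random value, feeding glass shard mask and normal textures.
# Reflection-vector-mapped noise gives a fake glint, divided by an alpha-driven value and offset by a random amount.
# Shading path switches replace both random inputs with constants on mobile.
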Material /BRCosmetics/Weapons/FORT_Melee/Pickaxe_Frost_Mystery_1h/FX/M_Shatter_GPU
graph {
 properties {
  bUsedWithNiagaraSprites=true
  bUsedWithParticleSprites=true
  blendMode=BLEND_Masked
  pixelDepthOffsetMode=PDOM_Legacy
  usedShadingModels=Default Lit
 }
 nodes [
  n1 = ComponentMask(None=ReflectionVectorWS()) {g=true, r=true}
  n2 = ParticleColor()
  n3 = ParticleRandom()
  Add(
   A=Divide(
    A=TextureSample(UVs=@n1) {samplerType=SAMPLERTYPE_LinearColor, texture=&HighContrastNoise}.RGB
    B=LinearInterpolate(Alpha=@n2.A) {constB=5}
   ) {constB=5}
   B=ShadingPathSwitch(Default=LinearInterpolate(Alpha=@n3) {constA=0.01, constB=0.2}, Mobile=Constant()) {desc=Migrated from FeatureLevelSwitch}
  )
  Constant() {r=0.1}
  Constant() {r=1}
  n4 = Constant() {r=2}
  n5 = MaterialFunctionCall(
   "Animation  Phase (0-1)"=ShadingPathSwitch(Default=LinearInterpolate(Alpha=@n3) {constA=0.1, constB=0.9}, Mobile=Constant()) {desc=Migrated from FeatureLevelSwitch}
   Number of Rows=@n4
   Number of Columns=@n4
  ) {materialFunction=&FlipBook}
  TextureSample(UVs=@n5.UVs) {texture=/Game/Effects/Fort_Effects/Textures/Glass/T_Cine_GlassShard_Mask}
  TextureSample(UVs=@n5.UVs) {samplerType=SAMPLERTYPE_Normal, texture=/Game/Effects/Fort_Effects/Textures/Glass/T_Cine_GlassShard}
 ]
}
